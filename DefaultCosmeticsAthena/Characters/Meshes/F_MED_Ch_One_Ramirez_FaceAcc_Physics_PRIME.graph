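# Physics asset for the female medium Ramirez character's face accessory mesh, providing bodies and joints for simulated hair and ponytail motion.
# Capsules on the upper spine, neck and head form the body chain, joined by free angular constraints.
# A four-link ponytail chain with tight twist limits hangs from the head.
# Separate hair strand bodies also attach to the head with angular limits, except testdyn_hair_05, which rotates freely.
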PhysicsAsset /DefaultCosmeticsAthena/Characters/Meshes/F_MED_Ch_One_Ramirez_FaceAcc_Physics_PRIME
bodies {
 dyn_hairPart_02 [{shapeType=Capsule, length=3.0743, radius=1.9454, shapeName=dyn_hairPart_02_capsule, center=<0 0 -2.6286>}]
 dyn_hairPart_03 [{shapeType=Capsule, length=2.6972, radius=1.2038, shapeName=dyn_hairPart_03_capsule, center=<0 0 -2.6286>}]
 dyn_hair_05 [{shapeType=Capsule, length=9.548, radius=4.6777, center=<9.2672 -1.6734 -2.2785>, rotation=<rot 0 -83.1806 -50.1641>}]
 dyn_ponytail_05 [{shapeType=Capsule, length=5.559, radius=2.9861, shapeName=dyn_ponytail_05_capsule, center=<2.7879 0 0>, rotation=<rot 90 0 0>}]
 dyn_ponytail_05_a [{shapeType=Capsule, length=5.3307, radius=3.353, shapeName=dyn_ponytail_05_a_capsule, center=<2.7879 0 0>, rotation=<rot 90 0 0>}]
 dyn_ponytail_05_b [{shapeType=Capsule, length=2.6653, radius=2.3617, shapeName=dyn_ponytail_05_b_capsule, center=<2.7879 0 0>, rotation=<rot 90 0 0>}]
 dyn_ponytail_05_c [{shapeType=Capsule, length=1.697, radius=1.8933, shapeName=dyn_ponytail_05_c_capsule, center=<3.4536 0 0>, rotation=<rot 90 0 0>}]
 head [{shapeType=Capsule, length=6.7787, radius=8.7735, center=<2.5279 0.5474 0>, rotation=<rot 90 0 0>}]
 neck_01 [{shapeType=Capsule, length=2.625, radius=5.505, shapeName=neck_01_capsule, rotation=<rot 90 0 0>}]
 neck_02 [{shapeType=Capsule, length=2.625, radius=5.505, shapeName=neck_02_capsule, rotation=<rot 90 0 0>}]
 spine_05 [
  {shapeType=Capsule, length=20.125, radius=9.875, center=<5.2065 3.4918 0>}
  {shapeType=Capsule, length=10.875, radius=10.625, shapeName=spine_05_capsule, center=<2.3854 -0.1965 0>, rotation=<rot 90 0 -10.0051>}
 ]
 testdyn_hair_05 [{
   shapeType=Capsule
   length=5.059
   radius=3.4861
   shapeName=testdyn_hair_05_capsule
   center=<4.3751 -2.2412 1.1077>
   rotation=<rot 0 -173.2511 -52.9796>
  }]
}
constraints [
 {bone1Name=dyn_hair_05, bone2Name=head, swing1LimitDegrees=45, swing2LimitDegrees=45, twistLimitDegrees=45}
 {
  bone1Name=testdyn_hair_05
  bone2Name=head
  swing1LimitDegrees=45
  swing1Motion=Free
  swing2LimitDegrees=45
  swing2Motion=Free
  twistLimitDegrees=45
  twistMotion=Free
 }
 {bone1Name=dyn_ponytail_05, bone2Name=head, swing1LimitDegrees=45, swing2LimitDegrees=45, twistLimitDegrees=2}
 {bone1Name=dyn_ponytail_05_a, bone2Name=dyn_ponytail_05, swing1LimitDegrees=45, swing2LimitDegrees=45, twistLimitDegrees=2}
 {bone1Name=dyn_ponytail_05_b, bone2Name=dyn_ponytail_05_a, swing1LimitDegrees=45, swing2LimitDegrees=45, twistLimitDegrees=2}
 {bone1Name=dyn_ponytail_05_c, bone2Name=dyn_ponytail_05_b, swing1LimitDegrees=45, swing2LimitDegrees=45, twistLimitDegrees=2}
 {
  bone1Name=head
  bone2Name=neck_02
  swing1LimitDegrees=45
  swing1Motion=Free
  swing2LimitDegrees=45
  swing2Motion=Free
  twistLimitDegrees=45
  twistMotion=Free
 }
 {
  bone1Name=neck_02
  bone2Name=neck_01
  swing1LimitDegrees=45
  swing1Motion=Free
  swing2LimitDegrees=45
  swing2Motion=Free
  twistLimitDegrees=45
  twistMotion=Free
 }
 {
  bone1Name=neck_01
  bone2Name=spine_05
  swing1LimitDegrees=45
  swing1Motion=Free
  swing2LimitDegrees=45
  swing2Motion=Free
  twistLimitDegrees=45
  twistMotion=Free
 }
 {bone1Name=dyn_hairPart_02, bone2Name=head, swing1LimitDegrees=2, swing2LimitDegrees=15, twistLimitDegrees=15}
 {bone1Name=dyn_hairPart_03, bone2Name=head, swing1LimitDegrees=2, swing2LimitDegrees=15, twistLimitDegrees=15}
]
