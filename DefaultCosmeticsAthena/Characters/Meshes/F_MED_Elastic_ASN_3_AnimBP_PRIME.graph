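# Animation Blueprint for the F_MED_Elastic_ASN_3 PRIME cosmetic character mesh, adding posed and simulated ponytail motion to a copied pose.
# The attached parent's pose and curves are blended per bone at both upper arms with a copy from Body_Mesh, set to the owning component.
# Head yaw relative to spine_05 drives additive offsets on both ponytail base bones.
# A RigidBody node then simulates relative to spine_03 using an override physics asset.
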AnimBlueprint /DefaultCosmeticsAthena/Characters/Meshes/F_MED_Elastic_ASN_3_AnimBP_PRIME
target_skeleton=&/F_MED_Elastic_ASN_3_Skeleton
graphs [
 AnimGraph {nodes [
   n1 = Copy Pose From Mesh(SourceMeshComponent=Get Body_Mesh())
   n2 = Copy Pose From Mesh() {bCopyCurves=true, bUseAttachedParent=true}
   n3 = Get headUp() -> (headUp:double)
   n4 = "Transform (Modify) Bone - Bone: asn3_ponytail_a_base_l"(
    ComponentPose=Local To Component(LocalPose=Layered blend per bone(BasePose=@n2, BlendPoses_0=@n1) {layerSetup [{branchFilters [{boneName=upperarm_l}, {boneName=upperarm_r}]}]})
    Alpha=@n3
    Translation="0,-1,0"
    Rotation="0,-50,0"
    Scale="1,1,1"
   ) {
    rotationMode=BMM_Additive
    rotationSpace=BCS_BoneSpace
    translationMode=BMM_Additive
    translationSpace=BCS_BoneSpace
    alphaBoolBlend {blendOption=Linear}
    alphaScaleBiasClamp {bMapRange=true, inRange {max=0, min=-50}, outRange {max=0, min=1}}
    boneToModify {boneName=asn3_ponytail_a_base_l}
    rotation {yaw=-50}
    translation {y=-1}
   }
   n5 = "Transform (Modify) Bone - Bone: asn3_ponytail_a_base_r"(ComponentPose=@n4, Alpha=@n3, Translation="0,1,0", Rotation="0,-50,0", Scale="1,1,1") {
    rotationMode=BMM_Additive
    rotationSpace=BCS_BoneSpace
    translationMode=BMM_Additive
    translationSpace=BCS_BoneSpace
    alphaBoolBlend {blendOption=Linear}
    alphaScaleBiasClamp {bMapRange=true, inRange {max=0, min=-50}, outRange {max=0, min=1}}
    boneToModify {boneName=asn3_ponytail_a_base_r}
    rotation {yaw=-50}
    translation {y=1}
   }
   n6 = Get headUp()
   n7 = "Transform (Modify) Bone - Bone: asn3_ponytail_a_base_l"(ComponentPose=@n5, Alpha=@n6, Translation="0,0,0", Rotation="0,15,0", Scale="1,1,1") {
    rotationMode=BMM_Additive
    rotationSpace=BCS_BoneSpace
    translationMode=BMM_Additive
    translationSpace=BCS_BoneSpace
    alphaBoolBlend {blendOption=Linear}
    alphaScaleBiasClamp {bMapRange=true, inRange {max=15}}
    boneToModify {boneName=asn3_ponytail_a_base_l}
    rotation {yaw=15}
   }
   n8 = "Transform (Modify) Bone - Bone: asn3_ponytail_a_base_r"(ComponentPose=@n7, Alpha=@n6, Translation="0,0,0", Rotation="0,15,0", Scale="1,1,1") {
    rotationMode=BMM_Additive
    rotationSpace=BCS_BoneSpace
    translationMode=BMM_Additive
    translationSpace=BCS_BoneSpace
    alphaBoolBlend {blendOption=Linear}
    alphaScaleBiasClamp {bMapRange=true, inRange {max=15}}
    boneToModify {boneName=asn3_ponytail_a_base_r}
    rotation {yaw=15}
   }
   n9 = RigidBody(ComponentPose=@n8) {
    bClampLinearTranslationLimitToRefPose=true
    bOverrideWorldGravity=true
    overridePhysicsAsset=&/F_MED_Elastic_ASN_3_Physics_PRIME
    simulationSpace=BaseBoneSpace
    alphaBoolBlend {blendOption=Linear}
    baseBoneRef {boneName=spine_03}
    componentLinearAccScale=<0.4 0.4 0.4>
    componentLinearVelScale=<0.4 0.4 0.4>
    overrideWorldGravity {z=-980}
   }
   Output Pose(Result=Component To Local(ComponentPose=@n9)) {layerGroup=None}
  ]}
 EventGraph {nodes [
   Event BlueprintUpdateAnimation()
   > Sequence()
   n1 = GetDeltaTransformFromRefPose(self=GetOwnerSkeletalMeshComponent(), BoneName=head, BaseName=spine_05)
   > Set headUp(headUp:double=BreakRotator(InRot=BreakTransform(InTransform=@n1).Rotation).Yaw) -> (Output_Get:double)
   Event BlueprintInitializeAnimation()
   > Set Body_Mesh(Body_Mesh=GetOwnerSkeletalMeshComponent())
   > n2 = Cast To PlayerPawn_Athena(Object=TryGetPawnOwner())
   > n3 = Set Pawn Owner(Pawn Owner:Player Pawn Athena Object Reference=@n2) -> (Output_Get:Player Pawn Athena Object Reference)
   > Set Head_Mesh(Head_Mesh:Skeletal Mesh Component Object Reference=GetSkeletalMeshForPartType(self=@n3, PartType=Head)) -> (Output_Get:Skeletal Mesh Component Object Reference)
  ]}
]
variables [
 Body_Mesh {type=SkeletalMeshComponent}
 Head_Mesh {type=SkeletalMeshComponent}
 headUp {type=number}
 Pawn Owner {type=PlayerPawn_Athena_C}
]
