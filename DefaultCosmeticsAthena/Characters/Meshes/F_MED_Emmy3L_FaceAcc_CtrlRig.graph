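# Control Rig for the F_MED_Emmy3L face accessory mesh that procedurally adjusts one hair bone as the head moves.
# A spherical pose reader measures the head's Y axis against spine_05 within a falloff region, with debug drawing left enabled.
# Its output weights a Modify Transforms node that rotates and shifts dyn_hair_10.
# The other dyn_hair bones are in the hierarchy but are not driven here.
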ControlRigBlueprint /DefaultCosmeticsAthena/Characters/Meshes/F_MED_Emmy3L_FaceAcc_CtrlRig
graphs [
 RigVMFunctionLibrary {links [], nodes []}
 RigVMModel {nodes [
   RigUnit_BeginExecution = &RigUnit_BeginExecution()
   n1 = &RigUnit_SphericalPoseReader(
    ActiveRegionScaleFactors=(PositiveWidth=1,NegativeWidth=1,PositiveHeight=1,NegativeHeight=1)
    ActiveRegionSize=0.2
    Debug=(bDrawDebug=true,DebugScale=25,DebugSegments=20,DebugThickness=0.25)
    DriverAxis=(X=0,Y=1,Z=0)
    DriverItem=(Type=Bone,Name="head")
    FalloffRegionScaleFactors=(PositiveWidth=1,NegativeWidth=1,PositiveHeight=1,NegativeHeight=1)
    FalloffSize=0.5
    OptionalParentItem=(Type=Bone,Name="spine_05")
    RotationOffset=(X=0,Y=-90,Z=0)
    ExecutePin=@RigUnit_BeginExecution.ExecutePin
   )
   > &RigUnit_ModifyTransforms(
    ItemToModify=((Item=(Type=Bone,Name="dyn_hair_10"),Transform=(Rotation=(X=0,Y=0,Z=0,W=1),Translation=(X=0,Y=4,Z=0),Scale3D=(X=1,Y=1,Z=1))))
    ItemToModify.0.Transform.Rotation=&RigVMFunction_MathQuaternionFromEuler(Euler=(X=-40,Y=0,Z=0)).Result
    Weight=@n1.OutputParam
   )
  ]}
]
hierarchy [
 root {loc=<0 0 0>, rot=<0 0 0 1>}
 pelvis {parent=root, loc=<0 -0.5301 98.0571>, rot=<0.0127 0.707 -0.0127 -0.707>}
 spine_01 {parent=pelvis, loc=<6.1208 -0.9328 0>, rot=<0 0 0.034 -0.9994>}
 spine_02 {parent=spine_01, loc=<6.5996 -0.2095 0>, rot=<0 0 0.0238 -0.9997>}
 spine_03 {parent=spine_02, loc=<6.4914 -0.1237 0>, rot=<0 0 0.0149 -0.9999>}
 spine_04 {parent=spine_03, loc=<6.9598 0.0794 0>, rot=<0 0 -0.0132 -0.9999>}
 spine_05 {parent=spine_04, loc=<6.9987 0 0>, rot=<0 0 -0.0015 -1>}
 neck_01 {parent=spine_05, loc=<11.3138 0.1529 0>, rot=<0 0 -0.2003 -0.9797>}
 neck_02 {parent=neck_01, loc=<4.5627 0 0>, rot=<0 0 0.0504 -0.9987>}
 head {parent=neck_02, loc=<5.3914 0 0>, rot=<0 0 0.115 -0.9934>}
 dyn_hair_08 {parent=head, loc=<11.4896 -8.9347 -0.1354>, rot=<0.0029 -0.7071 -0.0029 -0.7071>}
 dyn_hair_07 {parent=head, loc=<14.9805 -5.6067 -3.7695>, rot=<0.0029 -0.7071 -0.0029 -0.7071>}
 dyn_hair_06 {parent=head, loc=<14.2487 4.7759 -1.4182>, rot=<0.0029 -0.7071 -0.0029 -0.7071>}
 dyn_hair_05 {parent=head, loc=<-0.9952 5.6484 -7.2109>, rot=<0.0029 -0.7071 -0.0029 -0.7071>}
 dyn_hair_04 {parent=head, loc=<-3.0928 2.2431 -7.9152>, rot=<0.0029 -0.7071 -0.0029 -0.7071>}
 dyn_hair_03 {parent=head, loc=<-3.0723 -0.2334 -8.396>, rot=<0.0029 -0.7071 -0.0029 -0.7071>}
 dyn_hair_02 {parent=head, loc=<-4.5136 -2.7487 -6.8955>, rot=<0.0029 -0.7071 -0.0029 -0.7071>}
 dyn_hair_01 {parent=head, loc=<4.6871 -1.1965 8.2067>, rot=<0.0029 -0.7071 -0.0029 -0.7071>}
 dyn_hair_09 {parent=head, loc=<-1.0639 -4.6298 4.5273>, rot=<0.0029 -0.7071 -0.0029 -0.7071>}
 dyn_hair_10 {parent=head, loc=<-0.7939 -9.0362 -1.6698>, rot=<0.0029 -0.7071 -0.0029 -0.7071>}
]
