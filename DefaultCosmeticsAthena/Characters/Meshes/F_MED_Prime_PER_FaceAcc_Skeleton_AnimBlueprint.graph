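# Animation Blueprint for the F_MED_Prime_PER_FaceAcc cosmetic mesh that adds physics-driven motion to its ponytail.
# It copies the attached parent's pose, applies two fixed additive offsets to dyn_ponytail_base, then runs a RigidBody simulation with an override physics asset in spine_05 space.
# The EventGraph maps head and left clavicle rotations relative to spine_05 into headBack and clavicleDelta, which the AnimGraph does not read.
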AnimBlueprint /DefaultCosmeticsAthena/Characters/Meshes/F_MED_Prime_PER_FaceAcc_Skeleton_AnimBlueprint
target_skeleton=&/F_MED_Prime_PER_FaceAcc_Skeleton
graphs [
 AnimGraph {nodes [
   n1 = "Transform (Modify) Bone - Bone: dyn_ponytail_base"(
    ComponentPose=Local To Component(LocalPose=Copy Pose From Mesh() {bUseAttachedParent=true})
    Alpha=1
    Translation="0,0,0"
    Rotation="-5,0,0"
    Scale="1,1,1"
   ) {
    rotationMode=BMM_Additive
    rotationSpace=BCS_BoneSpace
    translationMode=BMM_Additive
    translationSpace=BCS_BoneSpace
    alphaBoolBlend {blendOption=Linear}
    boneToModify {boneName=dyn_ponytail_base}
    rotation {pitch=-5}
   }
   n2 = "Transform (Modify) Bone - Bone: dyn_ponytail_base"(ComponentPose=@n1, Alpha=1, Translation="-0.5,0,0", Rotation="-10,0,0", Scale="1,1,1") {
    rotationMode=BMM_Additive
    rotationSpace=BCS_BoneSpace
    translationMode=BMM_Additive
    translationSpace=BCS_BoneSpace
    alphaBoolBlend {blendOption=Linear}
    boneToModify {boneName=dyn_ponytail_base}
    rotation {pitch=-10}
    translation {x=-0.5}
   }
   n3 = RigidBody(ComponentPose=@n2) {
    bClampLinearTranslationLimitToRefPose=true
    bOverrideWorldGravity=true
    overridePhysicsAsset=&/F_MED_Prime_PER_FaceAcc_Physics_PRIME
    simulationSpace=BaseBoneSpace
    alphaBoolBlend {blendOption=Linear}
    baseBoneRef {boneName=spine_05}
    componentAppliedLinearAccClamp=<6000 6000 6000>
    componentLinearAccScale=<0.25 0.25 0.25>
    componentLinearVelScale=<0.25 0.25 0.25>
    overrideWorldGravity {z=-980}
   }
   Output Pose(Result=Component To Local(ComponentPose=@n3)) {layerGroup=None}
  ]}
 EventGraph {nodes [
   Event BlueprintUpdateAnimation()
   n1 = GetOwnerSkeletalMeshComponent()
   n2 = BreakTransform(InTransform=GetDeltaTransformFromRefPose(self=@n1, BoneName=head, BaseName=spine_05))
   > Set headBack(headBack:double=MapRangeClamped(Value=BreakRotator(InRot=@n2.Rotation).Yaw, InRangeA=0, InRangeB=-45)) -> (Output_Get:double)
   n3 = BreakTransform(InTransform=GetDeltaTransformFromRefPose(self=@n1, BoneName=clavicle_l, BaseName=spine_05))
   > Set clavicleDelta(clavicleDelta:double=MapRangeClamped(Value=BreakRotator(InRot=@n3.Rotation).Pitch, InRangeA=0, InRangeB=45)) -> (Output_Get:double)
   TryGetPawnOwner()
  ]}
]
variables [
 clavicleDelta {type=number}
 headBack {type=number}
]
