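# Physics asset for the F_MED_Prime_S_ASN face accessory character mesh, setting up physics for a ponytail and left and right hair strands.
# A seven-bone ponytail and two five-bone side strands are capsule chains, each hung from a master body locked angularly to the head.
# Swing limits stay uniform along the ponytail and widen toward side-strand tips, with twist kept tight.
# Head, upper spine and clavicle shapes form the body volumes, joined by angularly free constraints.
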PhysicsAsset /DefaultCosmeticsAthena/Characters/Meshes/F_MED_Prime_S_ASN_FaceAcc_Physics
bodies {
 clavicle_l [{shapeType=Capsule, length=5, radius=7, shapeName=clavicle_l_capsule, center=<15 -1 -4>, rotation=<rot 70 0 0>}]
 clavicle_r [{shapeType=Capsule, length=5, radius=7, shapeName=clavicle_r_capsule, center=<-15 1 4>, rotation=<rot 70 0 0>}]
 dyn_l_hair_1 [{shapeType=Capsule, length=4, radius=1.2, shapeName=dyn_l_hair_1_capsule, center=<2 0 0>, rotation=<rot 90 0 0>}]
 dyn_l_hair_2 [{shapeType=Capsule, length=4, radius=1.2, shapeName=dyn_l_hair_2_capsule, center=<2 0 0>, rotation=<rot 90 0 0>}]
 dyn_l_hair_3 [{shapeType=Capsule, length=4, radius=1.2, shapeName=dyn_l_hair_3_capsule, center=<2 0 0>, rotation=<rot 90 0 0>}]
 dyn_l_hair_4 [{shapeType=Capsule, length=4, radius=1.2, shapeName=dyn_l_hair_4_capsule, center=<2 0 0>, rotation=<rot 90 0 0>}]
 dyn_l_hair_5 [{shapeType=Capsule, length=4, radius=1.2, shapeName=dyn_l_hair_5_capsule, center=<2 0 0>, rotation=<rot 90 0 0>}]
 dyn_l_hair_MSTR [{shapeType=Capsule, length=0.505, radius=0.505, shapeName=dyn_l_hair_MSTR_capsule, rotation=<rot 0 -90 0>}]
 dyn_ponytail_1 [{shapeType=Capsule, length=4, radius=2, shapeName=dyn_ponytail_1_capsule, center=<0.7069 0 0>, rotation=<rot 90 0 0>}]
 dyn_ponytail_2 [{shapeType=Capsule, length=4, radius=2, shapeName=dyn_ponytail_2_capsule, center=<2 0 0>, rotation=<rot 90 0 0>}]
 dyn_ponytail_3 [{shapeType=Capsule, length=4, radius=2, shapeName=dyn_ponytail_3_capsule, center=<2 0 0>, rotation=<rot 90 0 0>}]
 dyn_ponytail_4 [{shapeType=Capsule, length=4, radius=2, shapeName=dyn_ponytail_4_capsule, center=<2 0 0>, rotation=<rot 90 0 0>}]
 dyn_ponytail_5 [{shapeType=Capsule, length=4, radius=2, shapeName=dyn_ponytail_5_capsule, center=<2 0 0>, rotation=<rot 90 0 0>}]
 dyn_ponytail_6 [{shapeType=Capsule, length=4, radius=2, shapeName=dyn_ponytail_6_capsule, center=<2 0 0>, rotation=<rot 90 0 0>}]
 dyn_ponytail_7 [{shapeType=Capsule, length=4, radius=2, shapeName=dyn_ponytail_7_capsule, center=<-0.5095 0 -0.0001>, rotation=<rot 90 0 0>}]
 dyn_ponytail_MSTR [{shapeType=Capsule, length=0.505, radius=0.505, shapeName=dyn_ponytail_MSTR_capsule, rotation=<rot 0 -90 0>}]
 dyn_r_hair_1 [{shapeType=Capsule, length=4, radius=1.2, shapeName=dyn_r_hair_1_capsule, center=<2 0 0>, rotation=<rot 90 0 0>}]
 dyn_r_hair_2 [{shapeType=Capsule, length=4, radius=1.2, shapeName=dyn_r_hair_2_capsule, center=<2 0 0>, rotation=<rot 90 0 0>}]
 dyn_r_hair_3 [{shapeType=Capsule, length=4, radius=1.2, shapeName=dyn_r_hair_3_capsule, center=<2 0 0>, rotation=<rot 90 0 0>}]
 dyn_r_hair_4 [{shapeType=Capsule, length=4, radius=1.2, shapeName=dyn_r_hair_4_capsule, center=<2 0 0>, rotation=<rot 90 0 0>}]
 dyn_r_hair_5 [{shapeType=Capsule, length=4, radius=1.2, shapeName=dyn_r_hair_5_capsule, center=<2 0 0>, rotation=<rot 90 0 0>}]
 dyn_r_hair_MSTR [{shapeType=Capsule, length=0.505, radius=0.505, shapeName=dyn_r_hair_MSTR_capsule, rotation=<rot 0 -90 0>}]
 head [
  {
   shapeType=Capsule
   length=3.4358
   radius=5.8048
   shapeName=head_capsule
   center=<1.2915 3.1563 -0.2453>
   rotation=<rot 0.6652 90.5229 -89.6148>
  }
  {shapeType=Capsule, length=1.75, radius=6.75, shapeName=head_capsule, center=<0.9196 -1.1584 0>}
 ]
 spine_05 [
  {shapeType=Sphere, radius=7.75, shapeName=spine_05_sphere, center=<10.8995 -0.1539 -0.0001>}
  {shapeType=Capsule, length=17.005, radius=10.505, shapeName=spine_05_capsule, center=<1.4859 1.2562 0>, rotation=<rot 0 -89.9999 0>}
 ]
}
constraints [
 {bone1Name=dyn_ponytail_1, bone2Name=dyn_ponytail_MSTR, swing1LimitDegrees=20, swing2LimitDegrees=10, twistLimitDegrees=1}
 {bone1Name=dyn_ponytail_2, bone2Name=dyn_ponytail_1, swing1LimitDegrees=20, swing2LimitDegrees=10, twistLimitDegrees=1}
 {bone1Name=dyn_ponytail_3, bone2Name=dyn_ponytail_2, swing1LimitDegrees=20, swing2LimitDegrees=10, twistLimitDegrees=1}
 {bone1Name=dyn_ponytail_4, bone2Name=dyn_ponytail_3, swing1LimitDegrees=20, swing2LimitDegrees=10, twistLimitDegrees=1}
 {bone1Name=dyn_ponytail_5, bone2Name=dyn_ponytail_4, swing1LimitDegrees=20, swing2LimitDegrees=10, twistLimitDegrees=1}
 {bone1Name=dyn_ponytail_6, bone2Name=dyn_ponytail_5, swing1LimitDegrees=20, swing2LimitDegrees=10, twistLimitDegrees=1}
 {bone1Name=dyn_ponytail_7, bone2Name=dyn_ponytail_6, swing1LimitDegrees=20, swing2LimitDegrees=10, twistLimitDegrees=1}
 {bone1Name=dyn_l_hair_1, bone2Name=dyn_l_hair_MSTR, swing1LimitDegrees=1, swing2LimitDegrees=8, twistLimitDegrees=1, twistMotion=Locked}
 {bone1Name=dyn_l_hair_2, bone2Name=dyn_l_hair_1, swing1LimitDegrees=8, swing2LimitDegrees=10, twistLimitDegrees=1}
 {bone1Name=dyn_l_hair_3, bone2Name=dyn_l_hair_2, swing1LimitDegrees=14, swing2LimitDegrees=14, twistLimitDegrees=1}
 {bone1Name=dyn_l_hair_4, bone2Name=dyn_l_hair_3, swing1LimitDegrees=20, swing2LimitDegrees=14, twistLimitDegrees=1}
 {bone1Name=dyn_l_hair_5, bone2Name=dyn_l_hair_4, swing1LimitDegrees=20, swing2LimitDegrees=15, twistLimitDegrees=1}
 {bone1Name=dyn_r_hair_1, bone2Name=dyn_r_hair_MSTR, swing1LimitDegrees=1, swing2LimitDegrees=8, twistLimitDegrees=1}
 {bone1Name=dyn_r_hair_2, bone2Name=dyn_r_hair_1, swing1LimitDegrees=8, swing2LimitDegrees=10, twistLimitDegrees=1}
 {bone1Name=dyn_r_hair_3, bone2Name=dyn_r_hair_2, swing1LimitDegrees=14, swing2LimitDegrees=14, twistLimitDegrees=1}
 {bone1Name=dyn_r_hair_4, bone2Name=dyn_r_hair_3, swing1LimitDegrees=20, swing2LimitDegrees=14, twistLimitDegrees=1}
 {bone1Name=dyn_r_hair_5, bone2Name=dyn_r_hair_4, swing1LimitDegrees=20, swing2LimitDegrees=15, twistLimitDegrees=1}
 {
  bone1Name=dyn_ponytail_MSTR
  bone2Name=head
  swing1LimitDegrees=45
  swing1Motion=Locked
  swing2LimitDegrees=45
  swing2Motion=Locked
  twistLimitDegrees=45
  twistMotion=Locked
 }
 {
  bone1Name=dyn_l_hair_MSTR
  bone2Name=head
  swing1LimitDegrees=45
  swing1Motion=Locked
  swing2LimitDegrees=45
  swing2Motion=Locked
  twistLimitDegrees=45
  twistMotion=Locked
 }
 {
  bone1Name=dyn_r_hair_MSTR
  bone2Name=head
  swing1LimitDegrees=45
  swing1Motion=Locked
  swing2LimitDegrees=45
  swing2Motion=Locked
  twistLimitDegrees=45
  twistMotion=Locked
 }
 {
  bone1Name=head
  bone2Name=spine_05
  swing1LimitDegrees=45
  swing1Motion=Free
  swing2LimitDegrees=45
  swing2Motion=Free
  twistLimitDegrees=45
  twistMotion=Free
 }
 {
  bone1Name=clavicle_l
  bone2Name=spine_05
  swing1LimitDegrees=45
  swing1Motion=Free
  swing2LimitDegrees=45
  swing2Motion=Free
  twistLimitDegrees=45
  twistMotion=Free
 }
 {
  bone1Name=clavicle_r
  bone2Name=spine_05
  swing1LimitDegrees=45
  swing1Motion=Free
  swing2LimitDegrees=45
  swing2Motion=Free
  twistLimitDegrees=45
  twistMotion=Free
 }
]
