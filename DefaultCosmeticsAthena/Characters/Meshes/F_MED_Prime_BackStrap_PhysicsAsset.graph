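# Physics asset for the back strap on the F_MED_Prime character cosmetic, defining the bodies and joints for the strap's secondary motion.
# Capsules on the pelvis, upper spine and both thighs approximate the body around two strap capsules.
# The strap chain runs from spine_05 through a small master body to dyn_backStrap_01, whose joint has narrow, uneven swing and twist limits.
# The master-to-spine and spine-to-pelvis joints rotate freely, while the thighs have wider limits against the pelvis.
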PhysicsAsset /DefaultCosmeticsAthena/Characters/Meshes/F_MED_Prime_BackStrap_PhysicsAsset
bodies {
 dyn_backStrap_01 [{shapeType=Capsule, length=4.2194, radius=2.7348, shapeName=dyn_backStrap_01_capsule, center=<3.5 0 0>, rotation=<rot 90 0 0>}]
 dyn_backStrap_MSTR [{shapeType=Capsule, length=0.505, radius=0.505, shapeName=dyn_backStrap_MSTR_capsule, rotation=<rot 0 -90 0>}]
 pelvis [{
   shapeType=Capsule
   length=18.5901
   radius=10.5106
   shapeName=pelvis_capsule
   center=<-0.2818 -1.7665 -0.0186>
   rotation=<rot -1.7955 -90 -0.357>
  }]
 spine_05 [{shapeType=Capsule, length=13.3791, radius=9.1451, shapeName=spine_05_capsule}]
 thigh_l [{shapeType=Capsule, length=35.6924, radius=9, shapeName=thigh_l_capsule, center=<-25.7672 -2 0>, rotation=<rot 90 0 0>}]
 thigh_r [{shapeType=Capsule, length=35.8241, radius=9, shapeName=thigh_r_capsule, center=<25.8391 1 0>, rotation=<rot 90 0 0>}]
}
constraints [
 {bone1Name=thigh_l, bone2Name=pelvis, swing1LimitDegrees=45, swing2LimitDegrees=45, twistLimitDegrees=45}
 {bone1Name=thigh_r, bone2Name=pelvis, swing1LimitDegrees=45, swing2LimitDegrees=45, twistLimitDegrees=45}
 {
  bone1Name=dyn_backStrap_MSTR
  bone2Name=spine_05
  swing1LimitDegrees=45
  swing1Motion=Free
  swing2LimitDegrees=45
  swing2Motion=Free
  twistLimitDegrees=45
  twistMotion=Free
 }
 {bone1Name=dyn_backStrap_01, bone2Name=dyn_backStrap_MSTR, swing1LimitDegrees=10, swing2LimitDegrees=35, twistLimitDegrees=5}
 {
  bone1Name=spine_05
  bone2Name=pelvis
  swing1LimitDegrees=45
  swing1Motion=Free
  swing2LimitDegrees=45
  swing2Motion=Free
  twistLimitDegrees=45
  twistMotion=Free
 }
]
