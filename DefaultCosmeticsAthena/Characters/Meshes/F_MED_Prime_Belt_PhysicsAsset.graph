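# Physics asset for the F_MED_Prime_Belt character cosmetic mesh, defining bodies and joints for its dynamic belt bones.
# A small capsule on dyn_belt_MSTR joins to the pelvis with free swing and twist, leading a chain of three progressively longer boxes.
# The chain links have narrow twist limits, with swing tightest at the first link.
# Capsules on the pelvis and both thighs approximate the body around the belt.
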PhysicsAsset /DefaultCosmeticsAthena/Characters/Meshes/F_MED_Prime_Belt_PhysicsAsset
bodies {
 dyn_belt_01 [{shapeType=Box, extentX=3, extentY=1, extentZ=3.2, shapeName=dyn_belt_01_box, center=<0.75 0 0>}]
 dyn_belt_02 [{shapeType=Box, extentX=4, extentY=1, extentZ=3.2, shapeName=dyn_belt_02_box, center=<0.75 0 0>}]
 dyn_belt_03 [{shapeType=Box, extentX=5, extentY=1, extentZ=3.2, shapeName=dyn_belt_03_box, center=<0.75 0 0>}]
 dyn_belt_MSTR [{shapeType=Capsule, length=0.505, radius=0.505, shapeName=dyn_belt_MSTR_capsule, rotation=<rot 0 -90 0>}]
 pelvis [{
   shapeType=Capsule
   length=18.5901
   radius=10.5106
   shapeName=pelvis_capsule
   center=<-0.2818 -1.7665 -0.0186>
   rotation=<rot -1.7955 -90 -0.357>
  }]
 thigh_l [{shapeType=Capsule, length=35.6924, radius=9, shapeName=thigh_l_capsule, center=<-25.7672 -2 0>, rotation=<rot 90 0 0>}]
 thigh_r [{shapeType=Capsule, length=35.8241, radius=9, shapeName=thigh_r_capsule, center=<25.8391 1 0>, rotation=<rot 90 0 0>}]
}
constraints [
 {bone1Name=thigh_l, bone2Name=pelvis, swing1LimitDegrees=45, swing2LimitDegrees=45, twistLimitDegrees=45}
 {bone1Name=thigh_r, bone2Name=pelvis, swing1LimitDegrees=45, swing2LimitDegrees=45, twistLimitDegrees=45}
 {
  bone1Name=dyn_belt_MSTR
  bone2Name=pelvis
  swing1LimitDegrees=45
  swing1Motion=Free
  swing2LimitDegrees=45
  swing2Motion=Free
  twistLimitDegrees=45
  twistMotion=Free
 }
 {bone1Name=dyn_belt_01, bone2Name=dyn_belt_MSTR, swing1LimitDegrees=5, swing2LimitDegrees=5, twistLimitDegrees=2}
 {bone1Name=dyn_belt_02, bone2Name=dyn_belt_01, swing1LimitDegrees=15, swing2LimitDegrees=15, twistLimitDegrees=2}
 {bone1Name=dyn_belt_03, bone2Name=dyn_belt_02, swing1LimitDegrees=15, swing2LimitDegrees=15, twistLimitDegrees=2}
]
