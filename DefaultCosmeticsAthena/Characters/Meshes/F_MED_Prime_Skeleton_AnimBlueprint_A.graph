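# Post-process-style Anim Blueprint for the F_MED_Prime cosmetic skeleton, animating clothing and accessory secondary motion.
# Copied mesh pose runs through Control Rig, then weighted constraints place device, zipper, belt and strap bones.
# State- and head-driven bone offsets adjust belt, hood and back zippers around RigidBody armband, belt and strap simulation, followed by AnimDynamics zippers.
# EventGraph updates hood alphas from skydiving and head rotation, plus RigidBody gravity through a deprecated override function.
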AnimBlueprint /DefaultCosmeticsAthena/Characters/Meshes/F_MED_Prime_Skeleton_AnimBlueprint_A
target_skeleton=&/F_MED_Prime_Skeleton
graphs [
 AnimGraph {nodes [
   n1 = Control Rig(Source=Copy Pose From Mesh() {bCopyCurves=true, bUseAttachedParent=true}) {
    alphaBoolBlend {blendOption=Linear}
    controlRigAssetReference {blueprintRigClass=&/F_MED_Prime_CtrlRig.F_MED_Prime_CtrlRig_C}
   }
   n2 = "Constraint - dyn_device (PPP)"!(
    ComponentPose=Local To Component(LocalPose=@n1)
    Alpha=1
    ConstraintWeights_0=0.1439
    ConstraintWeights_1:float=0.7372
    ConstraintWeights_2:float=0.1189
   ) -> (Pose) {
    alphaBoolBlend {blendOption=Linear}
    boneToModify {boneName=dyn_device}
    constraintSetup [{targetBone {boneName=lowerarm_l}}, {targetBone {boneName=lowerarm_twist_01_l}}, {targetBone {boneName=lowerarm_twist_02_l}}]
    constraintWeights [1,1,1]
   }
   n3 = "Constraint - dyn_jacketZipper_MSTR (PPPPP)"!(
    ComponentPose=@n2
    Alpha=1
    ConstraintWeights_0=0.0383
    ConstraintWeights_1:float=0.1223
    ConstraintWeights_2:float=0.2458
    ConstraintWeights_3:float=0.1949
    ConstraintWeights_4:float=0.3986
   ) -> (Pose) {
    alphaBoolBlend {blendOption=Linear}
    boneToModify {boneName=dyn_jacketZipper_MSTR}
    constraintSetup [
     {targetBone {boneName=pelvis}}
     {targetBone {boneName=spine_01}}
     {targetBone {boneName=spine_02}}
     {targetBone {boneName=spine_03}}
     {targetBone {boneName=spine_04}}
    ]
    constraintWeights [1,1,1,1,1]
   }
   n4 = "Constraint - dyn_r_frontPantsZipper_MSTR (PPPPPPPPP)"!(
    ComponentPose=@n3
    Alpha=1
    ConstraintWeights_0=0.4432
    ConstraintWeights_1:float=0.0049
    ConstraintWeights_2:float=0
    ConstraintWeights_3:float=0.0185
    ConstraintWeights_4:float=0.0078
    ConstraintWeights_5:float=0.0405
    ConstraintWeights_6:float=0.3799
    ConstraintWeights_7:float=0.1026
    ConstraintWeights_8:float=0.0026
   ) -> (Pose) {
    alphaBoolBlend {blendOption=Linear}
    boneToModify {boneName=dyn_r_frontPantsZipper_MSTR}
    constraintSetup [
     {targetBone {boneName=pelvis}}
     {targetBone {boneName=spine_02}}
     {targetBone {boneName=calf_twist_02_l}}
     {targetBone {boneName=thigh_twist_01_l}}
     {targetBone {boneName=deform_thigh_fr_l}}
     {targetBone {boneName=thigh_r}}
     {targetBone {boneName=thigh_twist_01_r}}
     {targetBone {boneName=deform_thigh_fr_r}}
     {targetBone {boneName=deform_thigh_bk_r}}
    ]
    constraintWeights [1,1,1,1,1,1,1,1,1]
   }
   n5 = "Constraint - dyn_l_frontPantsZipper_MSTR (PPPPPPPPP)"!(
    ComponentPose=@n4
    Alpha=1
    ConstraintWeights_0=0.4311
    ConstraintWeights_1:float=0.0044
    ConstraintWeights_2:float=0.0425
    ConstraintWeights_3:float=0.3906
    ConstraintWeights_4:float=0.1029
    ConstraintWeights_5:float=0.0027
    ConstraintWeights_6:float=0
    ConstraintWeights_7:float=0.0183
    ConstraintWeights_8:float=0.0076
   ) -> (Pose) {
    alphaBoolBlend {blendOption=Linear}
    boneToModify {boneName=dyn_l_frontPantsZipper_MSTR}
    constraintSetup [
     {targetBone {boneName=pelvis}}
     {targetBone {boneName=spine_02}}
     {targetBone {boneName=thigh_l}}
     {targetBone {boneName=thigh_twist_01_l}}
     {targetBone {boneName=deform_thigh_fr_l}}
     {targetBone {boneName=deform_thigh_bk_l}}
     {targetBone {boneName=calf_twist_02_r}}
     {targetBone {boneName=thigh_twist_01_r}}
     {targetBone {boneName=deform_thigh_fr_r}}
    ]
    constraintWeights [1,1,1,1,1,1,1,1,1]
   }
   n6 = "Constraint - dyn_r_backPantsZipper_MSTR (PPPPPP)"!(
    ComponentPose=@n5
    Alpha=1
    ConstraintWeights_0=0.916
    ConstraintWeights_1:float=0.0107
    ConstraintWeights_2:float=0.0094
    ConstraintWeights_3:float=0.0024
    ConstraintWeights_4:float=0.0315
    ConstraintWeights_5:float=0.0299
   ) -> (Pose) {
    alphaBoolBlend {blendOption=Linear}
    boneToModify {boneName=dyn_r_backPantsZipper_MSTR}
    constraintSetup [
     {targetBone {boneName=pelvis}}
     {targetBone {boneName=spine_01}}
     {targetBone {boneName=spine_02}}
     {targetBone {boneName=deform_thigh_bk_l}}
     {targetBone {boneName=thigh_twist_01_r}}
     {targetBone {boneName=deform_thigh_bk_r}}
    ]
    constraintWeights [1,1,1,1,1,1]
   }
   n7 = "Constraint - dyn_l_backPantsZipper_MSTR (PPPPPPP)"!(
    ComponentPose=@n6
    Alpha=1
    ConstraintWeights_0=0.9187
    ConstraintWeights_1:float=0.0122
    ConstraintWeights_2:float=0.0106
    ConstraintWeights_3:float=0.0288
    ConstraintWeights_4:float=0.0251
    ConstraintWeights_5:float=0.0012
    ConstraintWeights_6:float=0.0034
   ) -> (Pose) {
    alphaBoolBlend {blendOption=Linear}
    boneToModify {boneName=dyn_l_backPantsZipper_MSTR}
    constraintSetup [
     {targetBone {boneName=pelvis}}
     {targetBone {boneName=spine_01}}
     {targetBone {boneName=spine_02}}
     {targetBone {boneName=thigh_twist_01_l}}
     {targetBone {boneName=deform_thigh_bk_l}}
     {targetBone {boneName=thigh_twist_01_r}}
     {targetBone {boneName=deform_thigh_bk_r}}
    ]
    constraintWeights [1,1,1,1,1,1,1]
   }
   n8 = "Constraint - dyn_r_lowPantsZipper_MSTR (PPPPP)"!(
    ComponentPose=@n7
    Alpha=1
    ConstraintWeights_0=0.0201
    ConstraintWeights_1:float=0.7959
    ConstraintWeights_2:float=0.1622
    ConstraintWeights_3:float=0.0198
    ConstraintWeights_4:float=0.002
   ) -> (Pose) {
    alphaBoolBlend {blendOption=Linear}
    boneToModify {boneName=dyn_r_lowPantsZipper_MSTR}
    constraintSetup [
     {targetBone {boneName=thigh_r}}
     {targetBone {boneName=calf_r}}
     {targetBone {boneName=calf_twist_02_r}}
     {targetBone {boneName=deform_knee_bk_01_r}}
     {targetBone {boneName=deform_knee_fr_01_r}}
    ]
    constraintWeights [1,1,1,1,1]
   }
   n9 = "Constraint - dyn_l_lowPantsZipper_MSTR (PPPPP)"!(
    ComponentPose=@n8
    Alpha=1
    ConstraintWeights_0=0.0577
    ConstraintWeights_1:float=0.7294
    ConstraintWeights_2:float=0.1552
    ConstraintWeights_3:float=0.0321
    ConstraintWeights_4:float=0.0257
   ) -> (Pose) {
    alphaBoolBlend {blendOption=Linear}
    boneToModify {boneName=dyn_l_lowPantsZipper_MSTR}
    constraintSetup [
     {targetBone {boneName=thigh_l}}
     {targetBone {boneName=calf_l}}
     {targetBone {boneName=calf_twist_02_l}}
     {targetBone {boneName=deform_knee_bk_01_l}}
     {targetBone {boneName=deform_knee_fr_01_l}}
    ]
    constraintWeights [1,1,1,1,1]
   }
   n10 = "Constraint - dyn_belt_MSTR (PPPP)"!(
    ComponentPose=@n9
    Alpha=1
    ConstraintWeights_0=0.7906
    ConstraintWeights_1:float=0.0709
    ConstraintWeights_2:float=0.1159
    ConstraintWeights_3:float=0.0226
   ) -> (Pose) {
    alphaBoolBlend {blendOption=Linear}
    boneToModify {boneName=dyn_belt_MSTR}
    constraintSetup [{targetBone {boneName=pelvis}}, {targetBone {boneName=spine_01}}, {targetBone {boneName=spine_02}}, {targetBone {boneName=spine_03}}]
    constraintWeights [1,1,1,1]
   }
   n11 = "Constraint - dyn_backStrap_MSTR (PPPPPPPP)"!(
    ComponentPose=@n10
    Alpha=1
    ConstraintWeights_0=0.0006
    ConstraintWeights_1:float=0.0017
    ConstraintWeights_2:float=0.1455
    ConstraintWeights_3:float=0.4805
    ConstraintWeights_4:float=0.0484
    ConstraintWeights_5:float=0.3144
    ConstraintWeights_6:float=0.0089
    ConstraintWeights_7:float=0
   ) -> (Pose) {
    alphaBoolBlend {blendOption=Linear}
    boneToModify {boneName=dyn_backStrap_MSTR}
    constraintSetup [
     {targetBone {boneName=spine_02}}
     {targetBone {boneName=spine_03}}
     {targetBone {boneName=spine_04}}
     {targetBone {boneName=spine_05}}
     {targetBone {boneName=clavicle_l}}
     {targetBone {boneName=clavicle_r}}
     {targetBone {boneName=upperarm_twist_01_r}}
     {targetBone {boneName=deform_pectoral_fr_r}}
    ]
    constraintWeights [1,1,1,1,1,1,1,1]
   }
   n12 = "Transform (Modify) Bone - Bone: dyn_belt_MSTR"!(ComponentPose=@n11, bAlphaBoolEnabled:bool=Get bIsCrouching(), Translation="0,0,0", Rotation="20,0,-10", Scale="1,1,1") -> (Pose) {
    alphaInputType=Bool
    rotationMode=BMM_Additive
    rotationSpace=BCS_BoneSpace
    translationMode=BMM_Additive
    translationSpace=BCS_BoneSpace
    alphaBoolBlend {blendInTime=0.1, blendOption=Linear, blendOutTime=0.1}
    boneToModify {boneName=dyn_belt_MSTR}
    rotation {pitch=20, roll=-10}
   }
   n13 = "Transform (Modify) Bone - Bone: dyn_backHood_MSTR"(ComponentPose=@n12, Alpha=Get Skydive Alpha() -> (Skydive Alpha:double), Translation="1.5,0,0", Rotation="0,0,0", Scale="1,0.8,1") {
    rotationMode=BMM_Additive
    rotationSpace=BCS_BoneSpace
    translationMode=BMM_Additive
    translationSpace=BCS_BoneSpace
    alphaBoolBlend {blendInTime=0.1, blendOption=Linear, blendOutTime=0.1}
    boneToModify {boneName=dyn_backHood_MSTR}
    scale {y=0.8}
    translation {x=1.5}
   }
   n14 = "Transform (Modify) Bone - Bone: dyn_backHood_MSTR"(ComponentPose=@n13, Alpha=Get Neck Alpha() -> (Neck Alpha:double), Translation="3.5,-2,0", Rotation="0,30,0", Scale="1,1,1") {
    rotationMode=BMM_Additive
    rotationSpace=BCS_BoneSpace
    translationMode=BMM_Additive
    translationSpace=BCS_BoneSpace
    alphaBoolBlend {blendInTime=0.1, blendOption=Linear, blendOutTime=0.1}
    boneToModify {boneName=dyn_backHood_MSTR}
    rotation {yaw=30}
    translation=<3.5 -2>
   }
   n15 = "Transform (Modify) Bone - Bone: dyn_backHood_MSTR"(ComponentPose=@n14, Alpha=Get Head Tilt Alpha() -> (Head Tilt Alpha:double), Translation="0,0,0", Rotation="20,0,0", Scale="1,1,1") {
    rotationMode=BMM_Additive
    rotationSpace=BCS_BoneSpace
    translationMode=BMM_Additive
    translationSpace=BCS_BoneSpace
    alphaBoolBlend {blendInTime=0.1, blendOption=Linear, blendOutTime=0.1}
    boneToModify {boneName=dyn_backHood_MSTR}
    rotation {pitch=20}
   }
   n16 = Get Gravity Override()
   n17 = RigidBody!(ComponentPose=@n15, Alpha, OverrideWorldGravity:Vector=@n16, ExternalForce) -> (Pose) {
    bClampLinearTranslationLimitToRefPose=true
    bOverrideWorldGravity=true
    overridePhysicsAsset=&/F_MED_Prime_Armband_PhysicsAsset
    simulationSpace=BaseBoneSpace
    alphaBoolBlend {blendOption=Linear}
    baseBoneRef {boneName=upperarm_r}
    componentAppliedLinearAccClamp=<3000 3000 3000>
    componentLinearAccScale=<0.5 0.2 0.125>
    componentLinearVelScale=<0.5 0.2 0.125>
    overrideWorldGravity {z=-980}
   }
   n18 = RigidBody(ComponentPose=@n17, OverrideWorldGravity=@n16) {
    bClampLinearTranslationLimitToRefPose=true
    bOverrideWorldGravity=true
    overridePhysicsAsset=&/F_MED_Prime_Belt_PhysicsAsset
    simulationSpace=BaseBoneSpace
    alphaBoolBlend {blendOption=Linear}
    baseBoneRef {boneName=pelvis}
    componentAppliedLinearAccClamp=<3000 3000 3000>
    componentLinearAccScale=<0.3 0.3 0.3>
    componentLinearVelScale=<0.3 0.3 0.3>
    overrideWorldGravity {z=-980}
   }
   n19 = RigidBody(ComponentPose=@n18, OverrideWorldGravity=@n16) {
    bClampLinearTranslationLimitToRefPose=true
    bOverrideWorldGravity=true
    overridePhysicsAsset=&/F_MED_Prime_BackStrap_PhysicsAsset
    simulationSpace=BaseBoneSpace
    alphaBoolBlend {blendOption=Linear}
    baseBoneRef {boneName=spine_04}
    componentAppliedLinearAccClamp=<3000 3000 3000>
    componentLinearAccScale=<0.3 0.3 0.4>
    componentLinearVelScale=<0.3 0.3 0.3>
    overrideWorldGravity {z=-980}
   }
   n20 = Get bIsJumping()
   n21 = "Transform (Modify) Bone - Bone: dyn_r_backPantsZipper_MSTR"!(ComponentPose=@n19, bAlphaBoolEnabled:bool=@n20, Translation="0,0,0", Rotation="0,-45,0", Scale="1,1,1") -> (Pose) {
    alphaInputType=Bool
    rotationMode=BMM_Additive
    rotationSpace=BCS_BoneSpace
    translationMode=BMM_Additive
    translationSpace=BCS_BoneSpace
    alphaBoolBlend {blendInTime=0.4, blendOption=Linear, blendOutTime=0.1}
    boneToModify {boneName=dyn_r_backPantsZipper_MSTR}
    rotation {yaw=-45}
   }
   n22 = "Transform (Modify) Bone - Bone: dyn_l_backPantsZipper_MSTR"!(ComponentPose=@n21, bAlphaBoolEnabled:bool=@n20, Translation="0,0,0", Rotation="0,-30,0", Scale="1,1,1") -> (Pose) {
    alphaInputType=Bool
    rotationMode=BMM_Additive
    rotationSpace=BCS_BoneSpace
    translationMode=BMM_Additive
    translationSpace=BCS_BoneSpace
    alphaBoolBlend {blendInTime=0.4, blendOption=Linear, blendOutTime=0.1}
    boneToModify {boneName=dyn_l_backPantsZipper_MSTR}
    rotation {yaw=-30}
   }
   n23 = Anim Dynamics - Bone: dyn_jacketZipper_01(ComponentPose=@n22, Alpha=1, ExternalForce="0,0,0") {
    angularDampingOverride=0.998
    bGravityOverrideInSimSpace=true
    bOverrideAngularDamping=true
    bUseGravityOverride=true
    simulationSpace=RootRelative
    alphaBoolBlend {blendOption=Linear}
    boundBone {boneName=dyn_jacketZipper_01}
    componentAppliedLinearAccClamp=<3000 3000 3000>
    componentLinearAccScale=<0.3 0.3 0.35>
    componentLinearVelScale=<0.3 0.3 0.3>
    gravityOverride {z=-980}
    physicsBodyDefinitions [{
      sphereCollisionRadius=0
      boundBone {boneName=dyn_jacketZipper_01}
      boxExtents=<7 3 3>
      constraintSetup {angularLimitsMax=<2 20 20>, angularLimitsMin=<-2 -20 -5>, angularTarget=<0 0 0>}
      localJointOffset=<1 0 0>
     }]
    relativeSpaceBone {boneName=dyn_jacketZipper_MSTR}
   }
   n24 = Anim Dynamics - Bone: dyn_r_lowPantsZipper_01(ComponentPose=@n23, Alpha=1, ExternalForce="0,0,0") {
    angularDampingOverride=0.998
    bGravityOverrideInSimSpace=true
    bOverrideAngularDamping=true
    bUseGravityOverride=true
    numSolverIterationsPreUpdate=2
    simulationSpace=RootRelative
    alphaBoolBlend {blendOption=Linear}
    boundBone {boneName=dyn_r_lowPantsZipper_01}
    componentAppliedLinearAccClamp=<3000 3000 3000>
    componentLinearAccScale=<0.3 0.3 0.3>
    componentLinearVelScale=<0.3 0.3 0.3>
    gravityOverride {z=-980}
    physicsBodyDefinitions [{
      sphereCollisionRadius=0
      boundBone {boneName=dyn_r_lowPantsZipper_01}
      boxExtents=<7 3 3>
      constraintSetup {angularLimitsMax=<2 15 15>, angularLimitsMin=<-2 -2 -15>, angularTarget=<0 0 0>}
      localJointOffset=<1 0 0>
     }]
    relativeSpaceBone {boneName=dyn_r_lowPantsZipper_MSTR}
   }
   n25 = Anim Dynamics - Bone: dyn_l_lowPantsZipper_01(ComponentPose=@n24, Alpha=1, ExternalForce="0,0,0") {
    angularDampingOverride=0.998
    bGravityOverrideInSimSpace=true
    bOverrideAngularDamping=true
    bUseGravityOverride=true
    numSolverIterationsPreUpdate=2
    simulationSpace=RootRelative
    alphaBoolBlend {blendOption=Linear}
    boundBone {boneName=dyn_l_lowPantsZipper_01}
    componentAppliedLinearAccClamp=<3000 3000 3000>
    componentLinearAccScale=<0.3 0.3 0.3>
    componentLinearVelScale=<0.3 0.3 0.3>
    gravityOverride {z=-980}
    physicsBodyDefinitions [{
      sphereCollisionRadius=0
      boundBone {boneName=dyn_l_lowPantsZipper_01}
      boxExtents=<8 4 4>
      constraintSetup {angularLimitsMax=<2 2 15>, angularLimitsMin=<-2 -20 -15>, angularTarget=<0 0 0>}
      localJointOffset=<1 0 0>
     }]
    relativeSpaceBone {boneName=dyn_l_lowPantsZipper_MSTR}
   }
   n26 = Anim Dynamics - Bone: dyn_r_backPantsZipper_01(ComponentPose=@n25, Alpha=1, ExternalForce="0,0,0") {
    angularDampingOverride=0.998
    angularSpringConstant=100
    bGravityOverrideInSimSpace=true
    bOverrideAngularDamping=true
    bUseGravityOverride=true
    numSolverIterationsPostUpdate=2
    numSolverIterationsPreUpdate=2
    simulationSpace=RootRelative
    alphaBoolBlend {blendOption=Linear}
    boundBone {boneName=dyn_r_backPantsZipper_01}
    componentAppliedLinearAccClamp=<3000 3000 3000>
    componentLinearAccScale=<0.3 0.3 0.35>
    componentLinearVelScale=<0.3 0.3 0.1>
    gravityOverride {z=-980}
    physicsBodyDefinitions [{
      sphereCollisionRadius=0
      boundBone {boneName=dyn_r_backPantsZipper_01}
      boxExtents=<9 4 4>
      constraintSetup {angularLimitsMax=<2 15 -5>, angularLimitsMin=<-2 -15 -75>, angularTarget=<0 0 -1>}
      localJointOffset=<1 0.1 0>
     }]
    relativeSpaceBone {boneName=pelvis}
   }
   n27 = Anim Dynamics - Bone: dyn_l_backPantsZipper_01(ComponentPose=@n26, Alpha=1, ExternalForce="0,0,0") {
    angularDampingOverride=0.998
    angularSpringConstant=100
    bGravityOverrideInSimSpace=true
    bOverrideAngularDamping=true
    bUseGravityOverride=true
    numSolverIterationsPostUpdate=2
    numSolverIterationsPreUpdate=2
    simulationSpace=RootRelative
    alphaBoolBlend {blendOption=Linear}
    boundBone {boneName=dyn_l_backPantsZipper_01}
    componentAppliedLinearAccClamp=<3000 3000 3000>
    componentLinearAccScale=<0.3 0.3 0.35>
    componentLinearVelScale=<0.3 0.3 0.1>
    gravityOverride {z=-980}
    physicsBodyDefinitions [{
      sphereCollisionRadius=0
      boundBone {boneName=dyn_l_backPantsZipper_01}
      boxExtents=<11 5 5>
      constraintSetup {angularLimitsMax=<2 15 -5>, angularLimitsMin=<-2 -15 -75>, angularTarget=<0 0 -1>}
      localJointOffset=<1 0.1 0>
     }]
    relativeSpaceBone {boneName=pelvis}
   }
   n28 = Anim Dynamics - Bone: dyn_r_frontPantsZipper_01(ComponentPose=@n27, Alpha=1, ExternalForce="0,0,0") {
    angularDampingOverride=0.998
    bGravityOverrideInSimSpace=true
    bOverrideAngularDamping=true
    bUseGravityOverride=true
    numSolverIterationsPreUpdate=2
    simulationSpace=RootRelative
    alphaBoolBlend {blendOption=Linear}
    boundBone {boneName=dyn_r_frontPantsZipper_01}
    componentAppliedLinearAccClamp=<3000 3000 3000>
    componentLinearAccScale=<0.3 0.3 0.3>
    componentLinearVelScale=<0.3 0.3 0.3>
    gravityOverride {z=-980}
    physicsBodyDefinitions [{
      sphereCollisionRadius=0
      boundBone {boneName=dyn_r_frontPantsZipper_01}
      boxExtents=<7 3 3>
      constraintSetup {angularLimitsMax=<2 15 15>, angularLimitsMin=<-2 -15 -2>, angularTarget=<0 0 0>}
      localJointOffset=<1 0 0>
     }]
    relativeSpaceBone {boneName=dyn_r_frontPantsZipper_MSTR}
   }
   n29 = Anim Dynamics - Bone: dyn_l_frontPantsZipper_01(ComponentPose=@n28, Alpha=1, ExternalForce="0,0,0") {
    angularDampingOverride=0.998
    bGravityOverrideInSimSpace=true
    bOverrideAngularDamping=true
    bUseGravityOverride=true
    numSolverIterationsPreUpdate=2
    simulationSpace=RootRelative
    alphaBoolBlend {blendOption=Linear}
    boundBone {boneName=dyn_l_frontPantsZipper_01}
    componentAppliedLinearAccClamp=<3000 3000 3000>
    componentLinearAccScale=<0.3 0.3 0.3>
    componentLinearVelScale=<0.3 0.3 0.3>
    gravityOverride {z=-980}
    physicsBodyDefinitions [{
      sphereCollisionRadius=0
      boundBone {boneName=dyn_l_frontPantsZipper_01}
      boxExtents=<8 4 4>
      constraintSetup {angularLimitsMax=<2 15 20>, angularLimitsMin=<-2 -15 -2>, angularTarget=<0 0 0>}
      localJointOffset=<1 0 0>
     }]
    relativeSpaceBone {boneName=dyn_l_frontPantsZipper_MSTR}
   }
   Output Pose(Result=Component To Local(ComponentPose=@n29)) {layerGroup=None}
  ]}
 EventGraph {nodes [
   Event BlueprintUpdateAnimation()
   > n1 = Sequence!(execute) -> (then_0, then_1, then_2:Exec)
   > n2 = Gravity Override State Function DEPRECATED(CharacterPartAnimInstance=Self-Reference(), GameTimeInSeconds=GetGameTimeInSeconds(), GravityOverrideParams=Get Gravity Override Params())
   > Set Gravity Override(Gravity Override=@n2)
   n3 = Branch(execute=@n1.then_2, Condition=OR Boolean(A=Get bIsSkydiveDiveMode(), B=Get bIsSkydiveFloating()))
   > Set Skydive Alpha(Skydive Alpha:double=1) -> (Output_Get:double)
   Set Skydive Alpha(execute=@n3.else, Skydive Alpha=0)
   n4 = BreakTransform(InTransform=GetDeltaTransformFromRefPose(self=GetOwnerSkeletalMeshComponent(), BoneName=head, BaseName=spine_05))
   Set Neck Alpha(execute=@n1.then_1, Neck Alpha:double=MapRangeClamped(Value=BreakRotator(InRot=@n4.Rotation).Yaw, InRangeA=-10, InRangeB=-40)) -> (Output_Get:double)
   n5 = BreakTransform(InTransform=GetDeltaTransformFromRefPose(self=GetOwnerSkeletalMeshComponent(), BoneName=head, BaseName=spine_05))
   > Set Head Tilt Alpha(Head Tilt Alpha:double=MapRangeClamped(Value=BreakRotator(InRot=@n5.Rotation).Pitch, InRangeA=0, InRangeB=30)) -> (Output_Get:double)
   TryGetPawnOwner()
  ]}
]
variables [
 Gravity Override {type=Vector}
 Gravity Override Params {type=GravityOverrideParamsStruct}
 Head Tilt Alpha {type=number}
 Neck Alpha {type=number}
 Skydive Alpha {type=number}
]
