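# Physics asset for the M_MED_Prime character cosmetic mesh, covering a dynamic belt and a few body bones with capsule shapes.
# The belt is a three-body chain from dyn_belt_MSTR through dyn_belt_01 to dyn_belt_02, with limited swing and twist between links.
# The belt master, both thighs and neck_01 each constrain to the pelvis with free swing and twist.
# The left thigh's capsule carries the name thigh_r_capsule.
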PhysicsAsset /DefaultCosmeticsAthena/Characters/Meshes/M_MED_Prime_Physics
bodies {
 dyn_belt_01 [{shapeType=Capsule, length=2.5322, radius=1.4678, shapeName=dyn_belt_01_capsule, center=<2.5527 0 0>, rotation=<rot 90 0 0>}]
 dyn_belt_02 [{shapeType=Capsule, length=2.5322, radius=1.4678, shapeName=dyn_belt_02_capsule, center=<2.5527 0 0>, rotation=<rot 90 0 0>}]
 dyn_belt_MSTR [{shapeType=Capsule, length=0.3084, radius=0.7946, shapeName=dyn_belt_MSTR_capsule, rotation=<rot 90 0 0>}]
 neck_01 [{shapeType=Capsule, length=11.1531, radius=4.7531, shapeName=neck_01_capsule}]
 pelvis [{shapeType=Capsule, length=7.588, radius=12.2835, shapeName=pelvis_capsule}]
 thigh_l [{shapeType=Capsule, length=29.2721, radius=11.6876, shapeName=thigh_r_capsule, center=<-19.5523 -0.0004 -0.0001>, rotation=<rot 90 0 0>}]
 thigh_r [{shapeType=Capsule, length=29.2721, radius=11.6876, shapeName=thigh_r_capsule, center=<18.9266 -0.0003 0>, rotation=<rot 90 0 0>}]
}
constraints [
 {bone1Name=dyn_belt_01, bone2Name=dyn_belt_MSTR, swing1LimitDegrees=25, swing2LimitDegrees=25, twistLimitDegrees=10}
 {bone1Name=dyn_belt_02, bone2Name=dyn_belt_01, swing1LimitDegrees=25, swing2LimitDegrees=25, twistLimitDegrees=10}
 {
  bone1Name=dyn_belt_MSTR
  bone2Name=pelvis
  swing1LimitDegrees=45
  swing1Motion=Free
  swing2LimitDegrees=45
  swing2Motion=Free
  twistLimitDegrees=45
  twistMotion=Free
 }
 {
  bone1Name=thigh_l
  bone2Name=pelvis
  swing1LimitDegrees=45
  swing1Motion=Free
  swing2LimitDegrees=45
  swing2Motion=Free
  twistLimitDegrees=45
  twistMotion=Free
 }
 {
  bone1Name=thigh_r
  bone2Name=pelvis
  swing1LimitDegrees=45
  swing1Motion=Free
  swing2LimitDegrees=45
  swing2Motion=Free
  twistLimitDegrees=45
  twistMotion=Free
 }
 {
  bone1Name=neck_01
  bone2Name=pelvis
  swing1LimitDegrees=45
  swing1Motion=Free
  swing2LimitDegrees=45
  swing2Motion=Free
  twistLimitDegrees=45
  twistMotion=Free
 }
]
