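# Physics asset for the M_MED_Prime_PASIFIKA face accessory mesh, a character cosmetic in DefaultCosmeticsAthena.
# It holds a single body on the dyn_hair_01 bone, made of one rotated capsule offset from the bone origin.
# It defines no constraints, so the asset itself links that body to no other.
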PhysicsAsset /DefaultCosmeticsAthena/Characters/Meshes/M_MED_Prime_PASIFIKA_FaceAcc_PhysicsAsset
bodies {dyn_hair_01 [{shapeType=Capsule, length=6.7867, radius=7.7131, center=<-6.493 0.1012 -1.2537>, rotation=<rot 1.8965 -89.9998 -12.7404>}]}
constraints []
