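# Physics asset for the M_MED_Stands_ASN face accessory mesh, defining capsule bodies and joints for its ponytail hair bones.
# A two-body base chain (dyn_hairPER2_02 to 03) branches from 03 into five two-segment strands, ponytails A through E.
# Each link is a constraint with the same tight swing and twist limit values, and every body uses a single small capsule.
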PhysicsAsset /DefaultCosmeticsAthena/Characters/Meshes/M_MED_Stands_ASN_FaceAcc_Physics
bodies {
 dyn_hairPER2_02 [{shapeType=Capsule, length=1, radius=1, shapeName=dyn_hairPER2_02_capsule}]
 dyn_hairPER2_03 [{shapeType=Capsule, length=3, radius=1, shapeName=dyn_hairPER2_03_capsule, center=<2 0 0>, rotation=<rot -90 0 0>}]
 dyn_hairPER2_ponytailA_01 [{shapeType=Capsule, length=3, radius=1, shapeName=dyn_hairPER2_ponytailA_01_capsule, center=<2 0 0>, rotation=<rot -90 0 0>}]
 dyn_hairPER2_ponytailA_02 [{shapeType=Capsule, length=3, radius=1, shapeName=dyn_hairPER2_ponytailA_02_capsule, center=<2 0 0>, rotation=<rot -90 0 0>}]
 dyn_hairPER2_ponytailB_01 [{shapeType=Capsule, length=3, radius=1, shapeName=dyn_hairPER2_ponytailB_01_capsule, center=<2 0 0>, rotation=<rot -90 0 0>}]
 dyn_hairPER2_ponytailB_02 [{shapeType=Capsule, length=3, radius=1, shapeName=dyn_hairPER2_ponytailB_02_capsule, center=<2 0 0>, rotation=<rot -90 0 0>}]
 dyn_hairPER2_ponytailC_01 [{shapeType=Capsule, length=3, radius=1, shapeName=dyn_hairPER2_ponytailC_01_capsule, center=<2 0 0>, rotation=<rot -90 0 0>}]
 dyn_hairPER2_ponytailC_02 [{shapeType=Capsule, length=3, radius=1, shapeName=dyn_hairPER2_ponytailC_02_capsule, center=<2 0 0>, rotation=<rot -90 0 0>}]
 dyn_hairPER2_ponytailD_01 [{shapeType=Capsule, length=3, radius=1, shapeName=dyn_hairPER2_ponytailD_01_capsule, center=<2 0 0>, rotation=<rot -90 0 0>}]
 dyn_hairPER2_ponytailD_02 [{shapeType=Capsule, length=3, radius=1, shapeName=dyn_hairPER2_ponytailD_02_capsule, center=<2 0 0>, rotation=<rot -90 0 0>}]
 dyn_hairPER2_ponytailE_01 [{shapeType=Capsule, length=3, radius=1, shapeName=dyn_hairPER2_ponytailE_01_capsule, center=<2 0 0>, rotation=<rot -90 0 0>}]
 dyn_hairPER2_ponytailE_02 [{shapeType=Capsule, length=3, radius=1, shapeName=dyn_hairPER2_ponytailE_02_capsule, center=<2 0 0>, rotation=<rot -90 0 0>}]
}
constraints [
 {bone1Name=dyn_hairPER2_ponytailA_01, bone2Name=dyn_hairPER2_03, swing1LimitDegrees=5, swing2LimitDegrees=5, twistLimitDegrees=5}
 {bone1Name=dyn_hairPER2_ponytailA_02, bone2Name=dyn_hairPER2_ponytailA_01, swing1LimitDegrees=5, swing2LimitDegrees=5, twistLimitDegrees=5}
 {bone1Name=dyn_hairPER2_ponytailB_01, bone2Name=dyn_hairPER2_03, swing1LimitDegrees=5, swing2LimitDegrees=5, twistLimitDegrees=5}
 {bone1Name=dyn_hairPER2_ponytailB_02, bone2Name=dyn_hairPER2_ponytailB_01, swing1LimitDegrees=5, swing2LimitDegrees=5, twistLimitDegrees=5}
 {bone1Name=dyn_hairPER2_ponytailC_01, bone2Name=dyn_hairPER2_03, swing1LimitDegrees=5, swing2LimitDegrees=5, twistLimitDegrees=5}
 {bone1Name=dyn_hairPER2_ponytailC_02, bone2Name=dyn_hairPER2_ponytailC_01, swing1LimitDegrees=5, swing2LimitDegrees=5, twistLimitDegrees=5}
 {bone1Name=dyn_hairPER2_ponytailD_01, bone2Name=dyn_hairPER2_03, swing1LimitDegrees=5, swing2LimitDegrees=5, twistLimitDegrees=5}
 {bone1Name=dyn_hairPER2_ponytailD_02, bone2Name=dyn_hairPER2_ponytailD_01, swing1LimitDegrees=5, swing2LimitDegrees=5, twistLimitDegrees=5}
 {bone1Name=dyn_hairPER2_ponytailE_01, bone2Name=dyn_hairPER2_03, swing1LimitDegrees=5, swing2LimitDegrees=5, twistLimitDegrees=5}
 {bone1Name=dyn_hairPER2_ponytailE_02, bone2Name=dyn_hairPER2_ponytailE_01, swing1LimitDegrees=5, swing2LimitDegrees=5, twistLimitDegrees=5}
 {bone1Name=dyn_hairPER2_03, bone2Name=dyn_hairPER2_02, swing1LimitDegrees=5, swing2LimitDegrees=5, twistLimitDegrees=5}
]
